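# Master character material for the M_MED_Bites male player body, built on material attributes for skeletal meshes with morph targets and mesh deformers.
# The shared PreFX character function feeds CharacterSkinFX, then MF_LiquidMetal, a function borrowed from the Bites pickaxe.
# An FX_Mask texture, black by default, supplies the liquid metal mask through a selectable channel that defaults to green.
# The result passes through the shared PostFX function and then a cinematic post-FX function.
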Material /BRCosmetics/Characters/Player/Male/Medium/Bodies/M_MED_Bites/Materials/M_MED_Bites_Body_MASTER
graph {
 properties {
  bAutomaticallySetUsageInEditor=false
  bUseMaterialAttributes=true
  bUsedWithMeshDeformer=true
  bUsedWithMorphTargets=true
  bUsedWithSkeletalMesh=true
  pixelDepthOffsetMode=PDOM_Legacy
  refractionMethod=RM_IndexOfRefraction
  shadingModel=MSM_FromMaterialExpression
  subsurfaceProfile=&SS_HeroSkin_02
  usedShadingModels=Default Lit | Subsurface | Clear Coat | Subsurface Profile | Cloth | Eye
 }
 nodes [
  n1 = MaterialFunctionCall() {materialFunction=&BaseFN_CharacterFunction_PreFX}
  n2 = TextureSampleParameter2D() {parameterName=FX_Mask, samplerType=SAMPLERTYPE_Masks, texture=&T_BlackMask}
  n3 = MaterialFunctionCall(
   AO=@n1.AO
   Distance Blend=@n1.AO
   Fresnel=@n1.AO
   Gmap=@n1.AO
   Pre Skinned Local Position=@n1.AO
   Material=MaterialFunctionCall!(
    material=MaterialFunctionCall(
     PreskinnedPosition=@n1.Pre skinned Local Position
     Gmap=@n1.Pre skinned Local Position
     MaterialAttributes=@n1.Pre skinned Local Position
     AO=@n1.Pre skinned Local Position
    ) {materialFunction=&CharacterSkinFX}.Result
    AO=@n1.AO
    PS_Position=@n1.AO
    FxMask=StaticComponentMaskParameter(None=@n2.RGB) {defaultG=true, parameterName=FXChannel}
   ) {materialFunction=/BRCosmetics/Weapons/FORT_Melee/Pickaxe_Bites/Material/MF_LiquidMetal}.material
  ) {materialFunction=&BaseFN_CharacterFunction_PostFX}
  MaterialFunctionCall(In=@n3.Result) {materialFunction=&CINE_MF_CharPostFX}
 ]
}
