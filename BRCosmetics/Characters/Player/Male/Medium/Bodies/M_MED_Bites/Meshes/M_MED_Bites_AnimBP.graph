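# Post-process-style AnimBlueprint for the Bites cosmetic body, layering secondary motion over the pose copied from the parent mesh.
# Weighted bone constraints and look-ats position accessory bones, then AnimDynamics swings the zipper, belts, ropes, pendant and ring.
# Thigh and wrist twist alphas drive corrective strap, groin and rope offsets.
# Right hand and forearm twist scale up for a visible Bites-tagged pickaxe, or follow the emote_BuiltIn curve during emotes.
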AnimBlueprint /BRCosmetics/Characters/Player/Male/Medium/Bodies/M_MED_Bites/Meshes/M_MED_Bites_AnimBP
target_skeleton=&/M_MED_Bites_Skeleton
graphs [
 AnimGraph {nodes [
   n1 = "Constraint - attach_cape (PP)"!(
    ComponentPose=Local To Component(LocalPose=Copy Pose From Mesh() {bUseAttachedParent=true})
    Alpha=1
    ConstraintWeights_0=0.9427
    ConstraintWeights_1:float=0.0573
   ) -> (Pose) {
    alphaBoolBlend {blendOption=Linear}
    boneToModify {boneName=attach_cape}
    constraintSetup [{targetBone {boneName=spine_05}}, {targetBone {boneName=spine_04}}]
    constraintWeights [1,1]
   }
   n2 = "Constraint - attach_backpack (PPP)"!(ComponentPose=@n1, Alpha=1, ConstraintWeights_0=0.6132, ConstraintWeights_1:float=0.2353, ConstraintWeights_2:float=0.1515) -> (Pose) {
    alphaBoolBlend {blendOption=Linear}
    boneToModify {boneName=attach_backpack}
    constraintSetup [{targetBone {boneName=spine_05}}, {targetBone {boneName=spine_04}}, {targetBone {boneName=spine_03}}]
    constraintWeights [1,1,1]
   }
   n3 = "Constraint - dyn_zipper_MSTR (PPPP)"!(
    ComponentPose=@n2
    Alpha=1
    ConstraintWeights_0=0.057
    ConstraintWeights_1:float=0.2395
    ConstraintWeights_2:float=0.3337
    ConstraintWeights_3:float=0.3698
   ) -> (Pose) {
    alphaBoolBlend {blendOption=Linear}
    boneToModify {boneName=dyn_zipper_MSTR}
    constraintSetup [{targetBone {boneName=spine_05}}, {targetBone {boneName=neck_01}}, {targetBone {boneName=neck_02}}, {targetBone {boneName=head}}]
    constraintWeights [1,1,1,1]
   }
   n4 = "Constraint - dyn_belt_l_01_MSTR (PPPPP)"!(
    ComponentPose=@n3
    Alpha=1
    ConstraintWeights_0=0.092
    ConstraintWeights_1:float=0.3895
    ConstraintWeights_2:float=0.3385
    ConstraintWeights_3:float=0.1725
    ConstraintWeights_4:float=0.0075
   ) -> (Pose) {
    alphaBoolBlend {blendOption=Linear}
    boneToModify {boneName=dyn_belt_l_01_MSTR}
    constraintSetup [
     {targetBone {boneName=spine_05}}
     {targetBone {boneName=spine_04}}
     {targetBone {boneName=spine_03}}
     {targetBone {boneName=spine_02}}
     {targetBone {boneName=spine_01}}
    ]
    constraintWeights [1,1,1,1,1]
   }
   n5 = "Constraint - dyn_belt_l_02_MSTR (PPPPP)"!(
    ComponentPose=@n4
    Alpha=1
    ConstraintWeights_0=0.04
    ConstraintWeights_1:float=0.3466
    ConstraintWeights_2:float=0.3476
    ConstraintWeights_3:float=0.2461
    ConstraintWeights_4:float=0.0197
   ) -> (Pose) {
    alphaBoolBlend {blendOption=Linear}
    boneToModify {boneName=dyn_belt_l_02_MSTR}
    constraintSetup [
     {targetBone {boneName=spine_05}}
     {targetBone {boneName=spine_04}}
     {targetBone {boneName=spine_03}}
     {targetBone {boneName=spine_02}}
     {targetBone {boneName=spine_01}}
    ]
    constraintWeights [1,1,1,1,1]
   }
   n6 = "Constraint - dyn_belt_r_01_MSTR (PPPPP)"!(
    ComponentPose=@n5
    Alpha=1
    ConstraintWeights_0=0.092
    ConstraintWeights_1:float=0.3895
    ConstraintWeights_2:float=0.3385
    ConstraintWeights_3:float=0.1725
    ConstraintWeights_4:float=0.0075
   ) -> (Pose) {
    alphaBoolBlend {blendOption=Linear}
    boneToModify {boneName=dyn_belt_r_01_MSTR}
    constraintSetup [
     {targetBone {boneName=spine_05}}
     {targetBone {boneName=spine_04}}
     {targetBone {boneName=spine_03}}
     {targetBone {boneName=spine_02}}
     {targetBone {boneName=spine_01}}
    ]
    constraintWeights [1,1,1,1,1]
   }
   n7 = "Constraint - dyn_belt_r_02_MSTR (PPPPP)"!(
    ComponentPose=@n6
    Alpha=1
    ConstraintWeights_0=0.04
    ConstraintWeights_1:float=0.3466
    ConstraintWeights_2:float=0.3476
    ConstraintWeights_3:float=0.2461
    ConstraintWeights_4:float=0.0197
   ) -> (Pose) {
    alphaBoolBlend {blendOption=Linear}
    boneToModify {boneName=dyn_belt_r_02_MSTR}
    constraintSetup [
     {targetBone {boneName=spine_05}}
     {targetBone {boneName=spine_04}}
     {targetBone {boneName=spine_03}}
     {targetBone {boneName=spine_02}}
     {targetBone {boneName=spine_01}}
    ]
    constraintWeights [1,1,1,1,1]
   }
   n8 = "Constraint - dyn_belt_mid_MSTR (PPPPP)"!(
    ComponentPose=@n7
    Alpha=1
    ConstraintWeights_0=0.04
    ConstraintWeights_1:float=0.3466
    ConstraintWeights_2:float=0.3476
    ConstraintWeights_3:float=0.2461
    ConstraintWeights_4:float=0.0197
   ) -> (Pose) {
    alphaBoolBlend {blendOption=Linear}
    boneToModify {boneName=dyn_belt_mid_MSTR}
    constraintSetup [
     {targetBone {boneName=spine_05}}
     {targetBone {boneName=spine_04}}
     {targetBone {boneName=spine_03}}
     {targetBone {boneName=spine_02}}
     {targetBone {boneName=spine_01}}
    ]
    constraintWeights [1,1,1,1,1]
   }
   n9 = "Constraint - dyn_ring_MSTR (PPPPP)"!(
    ComponentPose=@n8
    Alpha=1
    ConstraintWeights_0=0.025
    ConstraintWeights_1:float=0.336
    ConstraintWeights_2:float=0.341
    ConstraintWeights_3:float=0.278
    ConstraintWeights_4:float=0.02
   ) -> (Pose) {
    alphaBoolBlend {blendOption=Linear}
    boneToModify {boneName=dyn_ring_MSTR}
    constraintSetup [
     {targetBone {boneName=spine_05}}
     {targetBone {boneName=spine_04}}
     {targetBone {boneName=spine_03}}
     {targetBone {boneName=spine_02}}
     {targetBone {boneName=spine_01}}
    ]
    constraintWeights [1,1,1,1,1]
   }
   n10 = "Constraint - dyn_bag_MSTR (PPPP)"!(
    ComponentPose=@n9
    Alpha=1
    ConstraintWeights_0=0.2651
    ConstraintWeights_1:float=0.4642
    ConstraintWeights_2:float=0.1182
    ConstraintWeights_3:float=0.1525
   ) -> (Pose) {
    alphaBoolBlend {blendOption=Linear}
    boneToModify {boneName=dyn_bag_MSTR}
    constraintSetup [{targetBone {boneName=pelvis}}, {targetBone {boneName=spine_01}}, {targetBone {boneName=spine_02}}, {targetBone {boneName=spine_03}}]
    constraintWeights [1,1,1,1]
   }
   n11 = "Constraint - dyn_shoulder_belt_r_MSTR (PPP)"!(ComponentPose=@n10, Alpha=1, ConstraintWeights_0=0.477, ConstraintWeights_1:float=0.012, ConstraintWeights_2:float=0.511) -> (Pose) {
    alphaBoolBlend {blendOption=Linear}
    boneToModify {boneName=dyn_shoulder_belt_r_MSTR}
    constraintSetup [{targetBone {boneName=spine_05}}, {targetBone {boneName=spine_04}}, {targetBone {boneName=clavicle_r}}]
    constraintWeights [1,1,1]
   }
   n12 = "Constraint - dyn_hand_pendant_MSTR (PPP)"!(ComponentPose=@n11, Alpha=1, ConstraintWeights_0=0.288, ConstraintWeights_1:float=0.362, ConstraintWeights_2:float=0.35) -> (Pose) {
    alphaBoolBlend {blendOption=Linear}
    boneToModify {boneName=dyn_hand_pendant_MSTR}
    constraintSetup [{targetBone {boneName=lowerarm_twist_01_r}}, {targetBone {boneName=hand_r}}, {targetBone {boneName=pinky_metacarpal_r}}]
    constraintWeights [1,1,1]
   }
   n13 = "Constraint - dyn_hand_ropes_MSTR (PPP)"!(ComponentPose=@n12, Alpha=1, ConstraintWeights_0=0.234, ConstraintWeights_1:float=0.353, ConstraintWeights_2:float=0.413) -> (Pose) {
    alphaBoolBlend {blendOption=Linear}
    boneToModify {boneName=dyn_hand_ropes_MSTR}
    constraintSetup [{targetBone {boneName=lowerarm_twist_01_r}}, {targetBone {boneName=hand_r}}, {targetBone {boneName=pinky_metacarpal_r}}]
    constraintWeights [1,1,1]
   }
   n14 = "Constraint - def_chest_bag_l (PPPPP)"!(
    ComponentPose=@n13
    Alpha=1
    ConstraintWeights_0=0.028
    ConstraintWeights_1:float=0.334
    ConstraintWeights_2:float=0.34
    ConstraintWeights_3:float=0.278
    ConstraintWeights_4:float=0.02
   ) -> (Pose) {
    alphaBoolBlend {blendOption=Linear}
    boneToModify {boneName=def_chest_bag_l}
    constraintSetup [
     {targetBone {boneName=spine_05}}
     {targetBone {boneName=spine_04}}
     {targetBone {boneName=spine_03}}
     {targetBone {boneName=spine_02}}
     {targetBone {boneName=spine_01}}
    ]
    constraintWeights [1,1,1,1,1]
   }
   n15 = "Constraint - def_chest_bag_r (PPPPP)"!(
    ComponentPose=@n14
    Alpha=1
    ConstraintWeights_0=0.028
    ConstraintWeights_1:float=0.334
    ConstraintWeights_2:float=0.34
    ConstraintWeights_3:float=0.278
    ConstraintWeights_4:float=0.02
   ) -> (Pose) {
    alphaBoolBlend {blendOption=Linear}
    boneToModify {boneName=def_chest_bag_r}
    constraintSetup [
     {targetBone {boneName=spine_05}}
     {targetBone {boneName=spine_04}}
     {targetBone {boneName=spine_03}}
     {targetBone {boneName=spine_02}}
     {targetBone {boneName=spine_01}}
    ]
    constraintWeights [1,1,1,1,1]
   }
   n16 = "Transform (Modify) Bone - Bone: dyn_hand_ropes_MSTR"(ComponentPose=@n15, Alpha=Get hand_rot_dn_r() -> (hand_rot_dn_r:double), Translation="0,0,0.5", Rotation="15,0,-15", Scale="1,1,1") {
    rotationMode=BMM_Additive
    rotationSpace=BCS_BoneSpace
    translationMode=BMM_Additive
    translationSpace=BCS_BoneSpace
    alphaBoolBlend {blendOption=Linear}
    boneToModify {boneName=dyn_hand_ropes_MSTR}
    rotation {pitch=15, roll=-15}
    translation {z=0.5}
   }
   n17 = Look At - Bone: dyn_belt_mid_AIM(ComponentPose=@n16, Alpha=1) {
    alphaBoolBlend {blendOption=Linear}
    boneToModify {boneName=dyn_belt_mid_AIM}
    lookAtLocation=<0 12.1214 -3.9187>
    lookAtTarget {boneReference {boneName=pelvis}}
    lookAt_Axis {axis=<1 0>}
    lookUp_Axis {axis {x=0}}
   }
   n18 = Look At - Bone: def_chest_bag_r(ComponentPose=@n17, Alpha=1) {
    alphaBoolBlend {blendOption=Linear}
    boneToModify {boneName=def_chest_bag_r}
    lookAtLocation=<0 10 9>
    lookAtTarget {boneReference {boneName=spine_01}}
    lookAt_Axis {axis=<1 0>}
    lookUp_Axis {axis {x=0}}
   }
   n19 = Look At - Bone: def_chest_bag_l(ComponentPose=@n18, Alpha=1) {
    alphaBoolBlend {blendOption=Linear}
    boneToModify {boneName=def_chest_bag_l}
    lookAtLocation=<0 10 -9>
    lookAtTarget {boneReference {boneName=spine_01}}
    lookAt_Axis {axis=<1 0>}
    lookUp_Axis {axis {x=0}}
   }
   n20 = Look At - Bone: dyn_zipper_AIM(ComponentPose=@n19, Alpha=1) {
    alphaBoolBlend {blendOption=Linear}
    boneToModify {boneName=dyn_zipper_AIM}
    lookAtLocation=<0 8.1>
    lookAtTarget {boneReference {boneName=spine_05}}
    lookAt_Axis {axis=<1 0>}
    lookUp_Axis {axis {x=0}}
   }
   n21 = Anim Dynamics - Bone: dyn_zipper(ComponentPose=@n20, Alpha=1) {
    angularDampingOverride=0.999
    bOverrideAngularDamping=true
    bUseGravityOverride=true
    simulationSpace=RootRelative
    alphaBoolBlend {blendOption=Linear}
    boundBone {boneName=dyn_zipper}
    componentAppliedLinearAccClamp=<3000 3000 3000>
    componentLinearAccScale=<0.3 0.3 0.3>
    componentLinearVelScale=<0.3 0.3 0.3>
    gravityOverride {z=-980}
    physicsBodyDefinitions [{
      boundBone {boneName=dyn_zipper}
      boxExtents=<10 16 8>
      constraintSetup {angularLimitsMax=<0 5 10>, angularLimitsMin=<0 -45 -10>, angularTarget=<0 0 0>}
      localJointOffset=<2 0 0>
     }]
   }
   n22 = Anim Dynamics - Bone: dyn_belt_mid_01(ComponentPose=@n21, Alpha=1, ExternalForce="0,0,0") {
    angularDampingOverride=0.999
    angularSpringConstant=500
    bAngularSpring=true
    bOverrideAngularDamping=true
    bUseGravityOverride=true
    gravityScale=-1
    simulationSpace=RootRelative
    alphaBoolBlend {blendOption=Linear}
    boundBone {boneName=dyn_belt_mid_01}
    componentAppliedLinearAccClamp=<3000 3000 3000>
    componentLinearAccScale=<0.3 0.3 0.3>
    componentLinearVelScale=<0.3 0.3 0.3>
    gravityOverride {z=-980}
    physicsBodyDefinitions [{
      boundBone {boneName=dyn_belt_mid_01}
      boxExtents=<12 16 8>
      constraintSetup {angularLimitsMax=<0 0 30>, angularLimitsMin=<0 -60 -30>, angularTarget=<0 0 0>}
      localJointOffset=<3 0 0>
     }]
   }
   n23 = Anim Dynamics - Bone: dyn_belt_l_01(ComponentPose=@n22, Alpha=1, ExternalForce="0,0,0") {
    angularDampingOverride=0.9999
    angularSpringConstant=500
    bOverrideAngularDamping=true
    bUseGravityOverride=true
    gravityScale=-1
    simulationSpace=RootRelative
    alphaBoolBlend {blendOption=Linear}
    boundBone {boneName=dyn_belt_l_01}
    componentAppliedLinearAccClamp=<3000 3000 3000>
    componentLinearAccScale=<0.3 0.3 0.3>
    componentLinearVelScale=<0.3 0.3 0.3>
    gravityOverride {z=-980}
    physicsBodyDefinitions [{
      boundBone {boneName=dyn_belt_l_01}
      boxExtents=<8 12 6>
      constraintSetup {angularLimitsMax=<0 5 10>, angularLimitsMin=<0 -25 -10>, angularTarget=<0 0 0>}
      localJointOffset=<1.2 1 0>
     }]
   }
   n24 = Anim Dynamics - Bone: dyn_belt_l_02(ComponentPose=@n23, Alpha=1, ExternalForce="0,0,0") {
    angularDampingOverride=0.9999
    angularSpringConstant=500
    bOverrideAngularDamping=true
    bUseGravityOverride=true
    gravityScale=-1
    simulationSpace=RootRelative
    alphaBoolBlend {blendOption=Linear}
    boundBone {boneName=dyn_belt_l_02}
    componentAppliedLinearAccClamp=<3000 3000 3000>
    componentLinearAccScale=<0.3 0.3 0.3>
    componentLinearVelScale=<0.3 0.3 0.3>
    gravityOverride {z=-980}
    physicsBodyDefinitions [{
      boundBone {boneName=dyn_belt_l_02}
      boxExtents=<8 12 6>
      constraintSetup {angularLimitsMax=<0 5 10>, angularLimitsMin=<0 -25 -10>, angularTarget=<0 0 0>}
      localJointOffset=<1.2 1.2 0>
     }]
   }
   n25 = Anim Dynamics - Bone: dyn_belt_r_01(ComponentPose=@n24, Alpha=1, ExternalForce="0,0,0") {
    angularDampingOverride=0.9999
    angularSpringConstant=500
    bOverrideAngularDamping=true
    bUseGravityOverride=true
    gravityScale=-1
    simulationSpace=RootRelative
    alphaBoolBlend {blendOption=Linear}
    boundBone {boneName=dyn_belt_r_01}
    componentAppliedLinearAccClamp=<3000 3000 3000>
    componentLinearAccScale=<0.3 0.3 0.3>
    componentLinearVelScale=<0.3 0.3 0.3>
    gravityOverride {z=-980}
    physicsBodyDefinitions [{
      boundBone {boneName=dyn_belt_r_01}
      boxExtents=<8 12 6>
      constraintSetup {angularLimitsMax=<0 5 10>, angularLimitsMin=<0 -25 -10>, angularTarget=<0 0 0>}
      localJointOffset=<1.2 -1 0>
     }]
   }
   n26 = Anim Dynamics - Bone: dyn_belt_r_02(ComponentPose=@n25, Alpha=1, ExternalForce="0,0,0") {
    angularDampingOverride=0.9999
    angularSpringConstant=500
    bOverrideAngularDamping=true
    bUseGravityOverride=true
    gravityScale=-1
    simulationSpace=RootRelative
    alphaBoolBlend {blendOption=Linear}
    boundBone {boneName=dyn_belt_r_02}
    componentAppliedLinearAccClamp=<3000 3000 3000>
    componentLinearAccScale=<0.3 0.3 0.3>
    componentLinearVelScale=<0.3 0.3 0.3>
    gravityOverride {z=-980}
    physicsBodyDefinitions [{
      boundBone {boneName=dyn_belt_r_02}
      boxExtents=<8 12 6>
      constraintSetup {angularLimitsMax=<0 5 10>, angularLimitsMin=<0 -25 -10>, angularTarget=<0 0 0>}
      localJointOffset=<1.2 -1 0>
     }]
   }
   n27 = Anim Dynamics - Bone: dyn_bag_belt_l(ComponentPose=@n26, Alpha=1, ExternalForce="0,0,0") {
    angularDampingOverride=0.99
    angularSpringConstant=500
    bAngularSpring=true
    bOverrideAngularDamping=true
    bUseGravityOverride=true
    gravityScale=-1
    simulationSpace=RootRelative
    alphaBoolBlend {blendOption=Linear}
    boundBone {boneName=dyn_bag_belt_l}
    componentAppliedLinearAccClamp=<3000 3000 3000>
    componentLinearAccScale=<0.3 0.3 0.3>
    componentLinearVelScale=<0.3 0.3 0.3>
    gravityOverride {z=-980}
    physicsBodyDefinitions [{
      boundBone {boneName=dyn_bag_belt_l}
      boxExtents=<12 16 8>
      constraintSetup {angularLimitsMax=<0 0 30>, angularLimitsMin=<0 -60 -30>, angularTarget=<0 0 0>}
      localJointOffset=<2 0 0>
     }]
   }
   n28 = Anim Dynamics - Bone: dyn_bag_belt_r(ComponentPose=@n27, Alpha=1, ExternalForce="0,0,0") {
    angularDampingOverride=0.99
    angularSpringConstant=500
    bAngularSpring=true
    bOverrideAngularDamping=true
    bUseGravityOverride=true
    gravityScale=-1
    simulationSpace=RootRelative
    alphaBoolBlend {blendOption=Linear}
    boundBone {boneName=dyn_bag_belt_r}
    componentAppliedLinearAccClamp=<3000 3000 3000>
    componentLinearAccScale=<0.3 0.3 0.3>
    componentLinearVelScale=<0.3 0.3 0.3>
    gravityOverride {z=-980}
    physicsBodyDefinitions [{
      boundBone {boneName=dyn_bag_belt_r}
      boxExtents=<12 16 8>
      constraintSetup {angularLimitsMax=<0 0 30>, angularLimitsMin=<0 -60 -30>, angularTarget=<0 0 0>}
      localJointOffset=<3 0 0>
     }]
   }
   n29 = Anim Dynamics - Bone: dyn_shoulder_belt_r(ComponentPose=@n28, Alpha=1, ExternalForce="0,0,0") {
    angularDampingOverride=1
    angularSpringConstant=500
    bOverrideAngularDamping=true
    bUseGravityOverride=true
    gravityScale=-1
    simulationSpace=RootRelative
    alphaBoolBlend {blendOption=Linear}
    boundBone {boneName=dyn_shoulder_belt_r}
    componentAppliedLinearAccClamp=<3000 3000 3000>
    componentLinearAccScale=<0.2 0.2 0.2>
    componentLinearVelScale=<0.3 0.3 0.3>
    gravityOverride {z=-980}
    physicsBodyDefinitions [{
      boundBone {boneName=dyn_shoulder_belt_r}
      boxExtents=<12 10 6>
      constraintSetup {angularLimitsMax=<0 25 15>, angularLimitsMin=<0 -35 -15>, angularTarget=<0 0 0>}
      localJointOffset=<0 0 -1.2>
     }]
   }
   n30 = Anim Dynamics - Bone: dyn_hand_pendant(ComponentPose=@n29, Alpha=1, ExternalForce="0,0,0") {
    angularDampingOverride=0.9999
    angularSpringConstant=500
    bAngularSpring=true
    bOverrideAngularDamping=true
    bUseGravityOverride=true
    gravityScale=-1
    simulationSpace=RootRelative
    alphaBoolBlend {blendOption=Linear}
    boundBone {boneName=dyn_hand_pendant}
    componentAppliedLinearAccClamp=<3000 3000 3000>
    componentLinearAccScale=<0.2 0.2 0.2>
    componentLinearVelScale=<0.3 0.3 0.3>
    gravityOverride {z=-980}
    physicsBodyDefinitions [{
      boundBone {boneName=dyn_hand_pendant}
      boxExtents=<10 8 10>
      constraintSetup {angularLimitsMax=<0 5 15>, angularLimitsMin=<0 -135 -10>, angularTarget=<0 0 0>}
      localJointOffset=<2 0 -0.4>
     }]
   }
   n31 = Anim Dynamics - Bone: dyn_hand_ropes_01(ComponentPose=@n30, Alpha=1, ExternalForce="0,0,0") {
    angularDampingOverride=0.999
    angularSpringConstant=500
    bAngularSpring=true
    bOverrideAngularDamping=true
    bUseGravityOverride=true
    gravityScale=-1
    simulationSpace=RootRelative
    alphaBoolBlend {blendOption=Linear}
    boundBone {boneName=dyn_hand_ropes_01}
    componentAppliedLinearAccClamp=<3000 3000 3000>
    componentLinearAccScale=<0.3 0.3 0.3>
    componentLinearVelScale=<0.3 0.3 0.3>
    gravityOverride {z=-980}
    physicsBodyDefinitions [{
      boundBone {boneName=dyn_hand_ropes_01}
      boxExtents=<8 16 16>
      constraintSetup {angularLimitsMax=<0 5 30>, angularLimitsMin=<0 -15 -30>, angularTarget=<0 0 0>}
      localJointOffset=<1.5 2 -1>
     }]
   }
   n32 = Anim Dynamics - Bone: dyn_hand_ropes_02(ComponentPose=@n31, Alpha=1, ExternalForce="0,0,0") {
    angularDampingOverride=0.999
    angularSpringConstant=500
    bAngularSpring=true
    bOverrideAngularDamping=true
    bUseGravityOverride=true
    gravityScale=-1
    simulationSpace=RootRelative
    alphaBoolBlend {blendOption=Linear}
    boundBone {boneName=dyn_hand_ropes_02}
    componentAppliedLinearAccClamp=<3000 3000 3000>
    componentLinearAccScale=<0.3 0.3 0.3>
    componentLinearVelScale=<0.3 0.3 0.3>
    gravityOverride {z=-980}
    physicsBodyDefinitions [{
      boundBone {boneName=dyn_hand_ropes_02}
      boxExtents=<8 16 16>
      constraintSetup {angularLimitsMax=<0 5 30>, angularLimitsMin=<0 -15 -30>, angularTarget=<0 0 0>}
      localJointOffset=<1.5 2 -1>
     }]
   }
   n33 = Anim Dynamics - Bone: dyn_ring(ComponentPose=@n32, Alpha=1, ExternalForce="0,0,0") {
    angularDampingOverride=0.9999
    angularSpringConstant=500
    bAngularSpring=true
    bOverrideAngularDamping=true
    bUseGravityOverride=true
    gravityScale=-1
    simulationSpace=RootRelative
    alphaBoolBlend {blendOption=Linear}
    boundBone {boneName=dyn_ring}
    componentAppliedLinearAccClamp=<3000 3000 3000>
    componentLinearAccScale=<0.2 0.2 0.2>
    componentLinearVelScale=<0.3 0.3 0.3>
    gravityOverride {z=-980}
    physicsBodyDefinitions [{
      boundBone {boneName=dyn_ring}
      boxExtents=<10 8 10>
      constraintSetup {angularLimitsMax=<0 0 0>, angularLimitsMin=<0 -60 0>, angularTarget=<0 0 0>}
      localJointOffset=<2 0 0>
     }]
   }
   n34 = Anim Dynamics - Bone: dyn_hand_belt_l(ComponentPose=@n33, Alpha=1, ExternalForce="0,0,0") {
    angularDampingOverride=0.9999
    angularSpringConstant=500
    bAngularSpring=true
    bOverrideAngularDamping=true
    bUseGravityOverride=true
    gravityScale=-1
    simulationSpace=RootRelative
    alphaBoolBlend {blendOption=Linear}
    boundBone {boneName=dyn_hand_belt_l}
    componentAppliedLinearAccClamp=<3000 3000 3000>
    componentLinearAccScale=<0.2 0.2 0.2>
    componentLinearVelScale=<0.3 0.3 0.3>
    gravityOverride {z=-980}
    physicsBodyDefinitions [{
      boundBone {boneName=dyn_hand_belt_l}
      boxExtents=<10 8 10>
      constraintSetup {angularLimitsMax=<0 15 15>, angularLimitsMin=<0 -15 -15>, angularTarget=<0 0 0>}
      localJointOffset=<-0.5 -1 -1>
     }]
   }
   n35 = Get def_front_strap_alpha() -> (def_front_strap_alpha:double)
   n36 = "Transform (Modify) Bone - Bone: def_ft_strap_r"(ComponentPose=@n34, Alpha=@n35, Translation="-7.8,9,-1.5", Rotation="0,-45,30", Scale="1,1,1") {
    rotationMode=BMM_Additive
    rotationSpace=BCS_BoneSpace
    scaleMode=BMM_Additive
    scaleSpace=BCS_BoneSpace
    translationMode=BMM_Additive
    translationSpace=BCS_BoneSpace
    alphaBoolBlend {blendOption=Linear}
    boneToModify {boneName=def_ft_strap_r}
    rotation {roll=30, yaw=-45}
    translation=<-7.8 9 -1.5>
   }
   n37 = "Transform (Modify) Bone - Bone: def_ft_strap_l"(ComponentPose=@n36, Alpha=@n35, Translation="-2,0,0", Rotation="0,0,15", Scale="1,1,1") {
    rotationMode=BMM_Additive
    rotationSpace=BCS_BoneSpace
    scaleMode=BMM_Additive
    scaleSpace=BCS_BoneSpace
    translationMode=BMM_Additive
    translationSpace=BCS_BoneSpace
    alphaBoolBlend {blendOption=Linear}
    boneToModify {boneName=def_ft_strap_l}
    rotation {roll=15}
    translation {x=-2}
   }
   n38 = "Transform (Modify) Bone - Bone: def_groin_r"(ComponentPose=@n37, Alpha=Get thigh_rot_up_r() -> (thigh_rot_up_r:double), Translation="10,-6,0", Rotation="0,0,0", Scale="1,1,1") {
    rotationMode=BMM_Additive
    rotationSpace=BCS_BoneSpace
    scaleMode=BMM_Additive
    scaleSpace=BCS_BoneSpace
    translationMode=BMM_Additive
    translationSpace=BCS_BoneSpace
    alphaBoolBlend {blendOption=Linear}
    boneToModify {boneName=def_groin_r}
    translation=<10 -6>
   }
   n39 = "Transform (Modify) Bone - Bone: def_groin_l"(ComponentPose=@n38, Alpha=Get thigh_rot_up_l() -> (thigh_rot_up_l:double), Translation="-30,20,0", Rotation="0,0,0", Scale="1,1,1") {
    rotationMode=BMM_Additive
    rotationSpace=BCS_BoneSpace
    scaleMode=BMM_Additive
    scaleSpace=BCS_BoneSpace
    translationMode=BMM_Additive
    translationSpace=BCS_BoneSpace
    alphaBoolBlend {blendOption=Linear}
    boneToModify {boneName=def_groin_l}
    translation=<-30 20>
   }
   n40 = Get bites_pickaxe() -> (bites_pickaxe:double)
   n41 = "Transform (Modify) Bone - Bone: hand_r"(ComponentPose=@n39, Alpha=@n40, Translation="0,0,0", Rotation="0,0,0", Scale="0.4,0.4,0.4") {scaleMode=BMM_Additive, scaleSpace=BCS_BoneSpace, alphaBoolBlend {blendOption=Linear}, boneToModify {boneName=hand_r}, scale=<0.4 0.4 0.4>}
   n42 = "Transform (Modify) Bone - Bone: lowerarm_twist_01_r"(ComponentPose=@n41, Alpha=@n40, Translation="0,0,0", Rotation="0,0,0", Scale="0.6,0.6,0.6") {
    scaleMode=BMM_Additive
    scaleSpace=BCS_BoneSpace
    alphaBoolBlend {blendOption=Linear}
    boneToModify {boneName=lowerarm_twist_01_r}
    scale=<0.6 0.6 0.6>
   }
   Output Pose(Result=Component To Local(ComponentPose=@n42)) {layerGroup=None}
  ]}
 EventGraph {nodes [
   Event BlueprintUpdateAnimation()
   > n1 = Sequence()
   n2 = Get SkelMesh()
   > Set hand_rot_dn_r(hand_rot_dn_r:double=GetAlphaFromReferenceTransformDelta(
     SkeletalMeshComponent=@n2
     BoneName=hand_r
     BaseName=lowerarm_r
     DeltaComponentType=QuaternionTwist
     InputRange=(X=0,Y=-110)
     TwistAxis="0,1,0"
    )) -> (Output_Get:double)
   n3 = GetAlphaFromReferenceTransformDelta(SkeletalMeshComponent=@n2, BoneName=thigh_l, BaseName=pelvis, DeltaComponentType=QuaternionTwist, InputRange=(X=0,Y=100), TwistAxis="0,0,1")
   > Set thigh_rot_up_l(thigh_rot_up_l:double=@n3) -> (Output_Get:double)
   > Set def_front_strap_alpha(def_front_strap_alpha:double=float * float(
     A=@n3
     B=GetAlphaFromReferenceTransformDelta(
      SkeletalMeshComponent=@n2
      BoneName=thigh_r
      BaseName=pelvis
      DeltaComponentType=QuaternionTwist
      InputRange=(X=0,Y=100)
      OutputRange=(X=1,Y=0)
      TwistAxis="0,0,1"
     )
    ) {is=title K2Node_PromotableOperator}) -> (Output_Get:double)
   > Set thigh_rot_up_r(thigh_rot_up_r:double=GetAlphaFromReferenceTransformDelta(SkeletalMeshComponent=@n2, BoneName=thigh_r, BaseName=pelvis, DeltaComponentType=QuaternionTwist, InputRange=(X=0,Y=100), TwistAxis="0,0,1")) -> (Output_Get:double)
   n4 = Branch(execute=@n1.then_1, Condition=Get bIsPlayingEmote())
   > Set bites_pickaxe(bites_pickaxe:double=GetCurveValue(self=Get AnimInst() -> (AnimInst:Anim Instance Object Reference), CurveName=emote_BuiltIn)) -> (Output_Get:double)
   Set FortPawn(execute=@n4.else, FortPawn:Fort Pawn Object Reference=TryGetFortPawn()) -> (Output_Get:Fort Pawn Object Reference)
   n5 = Get FortPawn() -> (FortPawn:Fort Pawn Object Reference)
   > n6 = Is Valid(InputObject=@n5)
   n7 = Set weapon_visible(execute=@n6.Is Not Valid, weapon_visible:bool=false) -> (Output_Get:bool)
   n8 = Get weapon_visible() -> (weapon_visible:bool)
   n9 = Branch(execute=[@n14.then, @n7.then], Condition=@n8)
   n10 = Make Array("[0]"=(TagName="Cosmetics.MetaTags.Bites"))
   > n11 = Branch(Condition=AND Boolean(A=@n8, B=PickaxeItemDefinitionHasAnyMetaTag(Tags=@n10, bRequiresEquipped=true)))
   > Set bites_pickaxe(bites_pickaxe=1)
   Set bites_pickaxe(execute=[@n11.else, @n9.else], bites_pickaxe=0)
   n12 = Get CurrentWeapon!(self:Fort Pawn Object Reference=@n5) -> (CurrentWeapon:Fort Weapon Object Reference)
   n13 = IsWeaponHolstered!(self:Fort Pawn Object Reference=@n5) -> (ReturnValue:bool)
   n14 = Set weapon_visible(execute=@n6.Is Valid, weapon_visible=AND Boolean(A=IsValid(Object=@n12), B=NOT Boolean(A=@n13)))
   Event BlueprintInitializeAnimation()
   n15 = GetOwnerSkeletalMeshComponent()
   > Is Valid(InputObject=@n15)
   > Set SkelMesh(SkelMesh=@n15)
   > n16 = Cast To FortPlayerPawn(Object=GetOwningActor())
   > Is Valid(InputObject=@n16)
   > n17 = Set Pawn(Pawn:Fort Player Pawn Object Reference=@n16) -> (Output_Get:Fort Player Pawn Object Reference)
   n18 = GetAnimInstance(self=Get Mesh(self=@n17))
   > Is Valid(InputObject=@n18)
   > Set AnimInst(AnimInst:Anim Instance Object Reference=@n18) -> (Output_Get:Anim Instance Object Reference)
   WeaponDataHasMatchingGameplayTag!(
    self:Fort Weapon Object Reference=Get CurrentWeapon(self=Get FortPawn())
    TagToCheck:Gameplay Tag Structure=(TagName="Cosmetics.MetaTags.Bites")
   ) -> (ReturnValue:bool)
  ]}
]
variables [
 AnimInst {type=AnimInstance, default=null}
 bites_pickaxe {type=number}
 def_front_strap_alpha {type=number}
 FortPawn {type=FortPawn, default=null}
 hand_rot_dn_r {type=number}
 Pawn {type=FortPlayerPawn}
 SkelMesh {type=SkeletalMeshComponent}
 thigh_rot_up_l {type=number}
 thigh_rot_up_r {type=number}
 weapon_visible {type=boolean}
]
