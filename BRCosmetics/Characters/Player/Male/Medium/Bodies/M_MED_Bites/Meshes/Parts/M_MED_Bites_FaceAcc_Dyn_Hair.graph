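# Physics asset for the Bites male medium character's face accessory mesh, defining bodies for its dynamic hair bones.
# Three left and three right hair bones each carry an offset, rotated capsule, with mirrored sides sharing the same shape setup, alongside a small capsule on the head.
# Every hair body is constrained directly to the head rather than chained bone to bone, with identical swing limits and a tighter twist limit.
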PhysicsAsset /BRCosmetics/Characters/Player/Male/Medium/Bodies/M_MED_Bites/Meshes/Parts/M_MED_Bites_FaceAcc_Dyn_Hair
bodies {
 dyn_hair_l_01 [{shapeType=Capsule, length=3, radius=3, shapeName=dyn_hair_01_capsule, center=<2 0 0>, rotation=<rot 90 0 0>}]
 dyn_hair_l_02 [{shapeType=Capsule, length=3, radius=3, shapeName=dyn_hair_02_capsule, center=<5 0 0>, rotation=<rot 90 0 0>}]
 dyn_hair_l_03 [{shapeType=Capsule, length=3, radius=3, shapeName=dyn_hair_03_capsule, center=<3 0 0>, rotation=<rot 90 0 0>}]
 dyn_hair_r_01 [{shapeType=Capsule, length=3, radius=3, shapeName=dyn_hair_01_capsule, center=<2 0 0>, rotation=<rot 90 0 0>}]
 dyn_hair_r_02 [{shapeType=Capsule, length=3, radius=3, shapeName=dyn_hair_02_capsule, center=<5 0 0>, rotation=<rot 90 0 0>}]
 dyn_hair_r_03 [{shapeType=Capsule, length=3, radius=3, shapeName=dyn_hair_03_capsule, center=<3 0 0>, rotation=<rot 90 0 0>}]
 head [{shapeType=Capsule, length=1, radius=1}]
}
constraints [
 {bone1Name=dyn_hair_l_01, bone2Name=head, swing1LimitDegrees=45, swing2LimitDegrees=45, twistLimitDegrees=15}
 {bone1Name=dyn_hair_l_02, bone2Name=head, swing1LimitDegrees=45, swing2LimitDegrees=45, twistLimitDegrees=15}
 {bone1Name=dyn_hair_l_03, bone2Name=head, swing1LimitDegrees=45, swing2LimitDegrees=45, twistLimitDegrees=15}
 {bone1Name=dyn_hair_r_01, bone2Name=head, swing1LimitDegrees=45, swing2LimitDegrees=45, twistLimitDegrees=15}
 {bone1Name=dyn_hair_r_02, bone2Name=head, swing1LimitDegrees=45, swing2LimitDegrees=45, twistLimitDegrees=15}
 {bone1Name=dyn_hair_r_03, bone2Name=head, swing1LimitDegrees=45, swing2LimitDegrees=45, twistLimitDegrees=15}
]
